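# Physics asset for the GongFiddle female medium character body, a SkyDive variant that drives the suit's hanging panels around the hips and legs.
# Seven two-bone dyn_suit chains hang from base anchors locked to both spine_01 and pelvis, with narrow swing limits and locked twist.
# Lower-body capsules cover pelvis, spine, glutes, legs and feet, with doubled thigh capsules, limited hips and free knee, ankle and toe joints.
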PhysicsAsset /BRCosmetics/Characters/Player/Female/Medium/Bodies/F_MED_GongFiddle/Meshes/F_MED_GongFiddle_Physics_SkyDive
bodies {
 ball_l [{shapeType=Capsule, length=5.4478, radius=4.7142, shapeName=ball_l_capsule, center=<3.008 2.177 -0.7742>, rotation=<rot -4.3199 -90 31.292>}]
 ball_r [{shapeType=Capsule, length=5.4478, radius=4.7142, shapeName=ball_r_capsule, center=<-3.008 -2.177 0.7742>, rotation=<rot -4.3199 -90 31.292>}]
 base_suitBack [{shapeType=Capsule, length=0.505, radius=0.505, shapeName=base_suitBack_capsule}]
 base_suitSideBack_l [{shapeType=Capsule, length=0.505, radius=0.505, shapeName=base_suitSideBack_l_capsule}]
 base_suitSideBack_r [{shapeType=Capsule, length=0.505, radius=0.505, shapeName=base_suitSideBack_r_capsule}]
 base_suitSideFront_l [{shapeType=Capsule, length=0.505, radius=0.505, shapeName=base_suitSideFront_l_capsule}]
 base_suitSideFront_r [{shapeType=Capsule, length=0.505, radius=0.505, shapeName=base_suitSideFront_r_capsule}]
 base_suitSide_l [{shapeType=Capsule, length=0.505, radius=0.505, shapeName=base_suitSide_l_capsule}]
 base_suitSide_r [{shapeType=Capsule, length=0.505, radius=0.505, shapeName=base_suitSide_r_capsule}]
 calf_l [{shapeType=Capsule, length=25, radius=7, shapeName=calf_l_capsule, center=<-25 -1.7 0.5>, rotation=<rot 90 -5.8227 0>}]
 calf_r [{shapeType=Capsule, length=25, radius=7, shapeName=calf_r_capsule, center=<26 1.7 -0.5>, rotation=<rot 90 -5.8227 0>}]
 deform_glute_l [{
   shapeType=Capsule
   length=3.4775
   radius=4.273
   shapeName=deform_glute_l_capsule
   center=<4.0604 -3.7584 -2.1788>
   rotation=<rot 0 -90.0001 -30.7074>
  }]
 deform_glute_r [{
   shapeType=Capsule
   length=3.4775
   radius=4.273
   shapeName=deform_glute_r_capsule
   center=<-4.0604 3.7584 2.1788>
   rotation=<rot 0 -90.0001 149.2926>
  }]
 dyn_suitBack_01 [{
   shapeType=Capsule
   length=11.2678
   radius=7.3925
   shapeName=dyn_suitSideFront_02_l_capsule
   center=<7.952 -7.1596 0.1099>
   rotation=<rot -7.3435 0 -0.0003>
  }]
 dyn_suitBack_02 [{shapeType=Capsule, length=9.6902, radius=8.9701, shapeName=dyn_suitSideFront_02_l_capsule, center=<7.2613 -5.8512 0>}]
 dyn_suitSideBack_01_l [{
   shapeType=Capsule
   length=12.5989
   radius=6.0614
   shapeName=dyn_suitSideFront_02_l_capsule
   center=<7.952 -6.5131 0>
   rotation=<rot -7.3435 0 -0.0003>
  }]
 dyn_suitSideBack_01_r [{
   shapeType=Capsule
   length=13.1858
   radius=5.4744
   shapeName=dyn_suitSideFront_02_l_capsule
   center=<8.8045 -7.3038 0.0913>
   rotation=<rot -7.3435 0 -0.0003>
  }]
 dyn_suitSideBack_02_l [{
   shapeType=Capsule
   length=13.72
   radius=4.9402
   shapeName=dyn_suitSideFront_02_l_capsule
   center=<7.2639 -6.5453 -0.1104>
   rotation=<rot -1.138 0 -0.9259>
  }]
 dyn_suitSideBack_02_r [{
   shapeType=Capsule
   length=13.7737
   radius=4.8865
   shapeName=dyn_suitSideFront_02_l_capsule
   center=<7.2642 -6.6558 -0.1058>
   rotation=<rot -1.138 0 -0.9259>
  }]
 dyn_suitSideFront_01_l [{
   shapeType=Capsule
   length=13.0444
   radius=5.0221
   shapeName=dyn_suitSideFront_02_l_capsule
   center=<7.952 -5.7977 -0.0061>
   rotation=<rot 5.7016 0 -0.0005>
  }]
 dyn_suitSideFront_01_r [{
   shapeType=Capsule
   length=12.6262
   radius=6.0341
   shapeName=dyn_suitSideFront_02_l_capsule
   center=<7.9519 -8.404 0.0572>
   rotation=<rot -0.1629 0 -0.0001>
  }]
 dyn_suitSideFront_02_l [{
   shapeType=Capsule
   length=12.2545
   radius=5.6517
   shapeName=dyn_suitSideFront_02_l_capsule
   center=<9.465 -6.1282 0.0545>
   rotation=<rot 10.7552 0 -0.9424>
  }]
 dyn_suitSideFront_02_r [{
   shapeType=Capsule
   length=11.8822
   radius=6.024
   shapeName=dyn_suitSideFront_02_l_capsule
   center=<9.4653 -6.5666 0.1861>
   rotation=<rot -3.565 0 -0.9272>
  }]
 dyn_suitSide_01_l [{
   shapeType=Capsule
   length=12.1248
   radius=6.5355
   shapeName=dyn_suitSideFront_02_l_capsule
   center=<7.952 -6.1656 0.238>
   rotation=<rot -7.3435 0 -0.0003>
  }]
 dyn_suitSide_01_r [{shapeType=Capsule, length=11.8437, radius=6.8166, shapeName=dyn_suitSideFront_02_l_capsule, center=<6.2819 -6.6282 0>}]
 dyn_suitSide_02_l [{
   shapeType=Capsule
   length=13.5
   radius=5.1602
   shapeName=dyn_suitSideFront_02_l_capsule
   center=<9.4628 -5.5157 0>
   rotation=<rot -6.6548 0 -0.9321>
  }]
 dyn_suitSide_02_r [{
   shapeType=Capsule
   length=13.8796
   radius=4.7807
   shapeName=dyn_suitSideFront_02_l_capsule
   center=<9.4649 -5.4137 -0.0369>
   rotation=<rot -2.2874 0 -0.9265>
  }]
 foot_l [{
   shapeType=Capsule
   length=9.8449
   radius=8.5179
   shapeName=foot_l_capsule
   center=<-4.2407 4.1563 0.5157>
   rotation=<rot -1.3977 -175.9875 -87.7512>
  }]
 foot_r [{
   shapeType=Capsule
   length=9.8395
   radius=8.5066
   shapeName=foot_r_capsule
   center=<4.2118 -4.1732 -0.4557>
   rotation=<rot -1.3977 -175.7104 -87.8281>
  }]
 pelvis [{shapeType=Capsule, length=10.7791, radius=10.4711, center=<3.3821 -0.5468 0.2718>, rotation=<rot 4.4575 -90.0015 177.6591>}]
 spine_01 [{shapeType=Capsule, length=9.2209, radius=9.1739}]
 thigh_l [
  {shapeType=Capsule, length=29.9507, radius=7.579, center=<-16.8369 -1.8512 0.0429>, rotation=<rot 0.2482 -92.6907 -89.0955>}
  {shapeType=Capsule, length=30.9743, radius=8.2045, center=<-16.9984 0.986 -0.2397>, rotation=<rot 0.1349 -85.5737 -89.0716>}
 ]
 thigh_r [
  {shapeType=Capsule, length=29.9507, radius=7.2741, center=<16.8369 1.8512 -0.0429>, rotation=<rot 0.2482 -92.6907 90.9045>}
  {shapeType=Capsule, length=30.0849, radius=8.4579, center=<16.9415 -2.3075 -0.0933>, rotation=<rot 0.1718 -87.9117 90.922>}
 ]
}
constraints [
 {
  bone1Name=spine_01
  bone2Name=pelvis
  swing1LimitDegrees=45
  swing1Motion=Locked
  swing2LimitDegrees=45
  swing2Motion=Locked
  twistLimitDegrees=45
  twistMotion=Locked
 }
 {
  bone1Name=base_suitBack
  bone2Name=spine_01
  swing1LimitDegrees=45
  swing1Motion=Locked
  swing2LimitDegrees=45
  swing2Motion=Locked
  twistLimitDegrees=45
  twistMotion=Locked
 }
 {bone1Name=dyn_suitBack_01, bone2Name=base_suitBack, swing1LimitDegrees=5, swing2LimitDegrees=5, twistLimitDegrees=45, twistMotion=Locked}
 {bone1Name=dyn_suitBack_02, bone2Name=dyn_suitBack_01, swing1LimitDegrees=10, swing2LimitDegrees=5, twistLimitDegrees=45, twistMotion=Locked}
 {
  bone1Name=base_suitSideBack_l
  bone2Name=spine_01
  swing1LimitDegrees=45
  swing1Motion=Locked
  swing2LimitDegrees=45
  swing2Motion=Locked
  twistLimitDegrees=45
  twistMotion=Locked
 }
 {
  bone1Name=dyn_suitSideBack_01_l
  bone2Name=base_suitSideBack_l
  swing1LimitDegrees=5
  swing2LimitDegrees=2
  twistLimitDegrees=45
  twistMotion=Locked
 }
 {
  bone1Name=dyn_suitSideBack_02_l
  bone2Name=dyn_suitSideBack_01_l
  swing1LimitDegrees=8
  swing2LimitDegrees=3
  twistLimitDegrees=45
  twistMotion=Locked
 }
 {
  bone1Name=base_suitSideBack_r
  bone2Name=spine_01
  swing1LimitDegrees=45
  swing1Motion=Locked
  swing2LimitDegrees=45
  swing2Motion=Locked
  twistLimitDegrees=45
  twistMotion=Locked
 }
 {
  bone1Name=dyn_suitSideBack_01_r
  bone2Name=base_suitSideBack_r
  swing1LimitDegrees=5
  swing2LimitDegrees=5
  twistLimitDegrees=45
  twistMotion=Locked
 }
 {
  bone1Name=dyn_suitSideBack_02_r
  bone2Name=dyn_suitSideBack_01_r
  swing1LimitDegrees=10
  swing2LimitDegrees=5
  twistLimitDegrees=45
  twistMotion=Locked
 }
 {
  bone1Name=base_suitSideFront_l
  bone2Name=spine_01
  swing1LimitDegrees=45
  swing1Motion=Locked
  swing2LimitDegrees=45
  swing2Motion=Locked
  twistLimitDegrees=45
  twistMotion=Locked
 }
 {
  bone1Name=dyn_suitSideFront_01_l
  bone2Name=base_suitSideFront_l
  swing1LimitDegrees=8
  swing2LimitDegrees=8
  twistLimitDegrees=45
  twistMotion=Locked
 }
 {
  bone1Name=dyn_suitSideFront_02_l
  bone2Name=dyn_suitSideFront_01_l
  swing1LimitDegrees=12
  swing2LimitDegrees=2
  twistLimitDegrees=45
  twistMotion=Locked
 }
 {
  bone1Name=base_suitSideFront_r
  bone2Name=spine_01
  swing1LimitDegrees=45
  swing1Motion=Locked
  swing2LimitDegrees=45
  swing2Motion=Locked
  twistLimitDegrees=45
  twistMotion=Locked
 }
 {
  bone1Name=dyn_suitSideFront_01_r
  bone2Name=base_suitSideFront_r
  swing1LimitDegrees=8
  swing2LimitDegrees=8
  twistLimitDegrees=45
  twistMotion=Locked
 }
 {
  bone1Name=dyn_suitSideFront_02_r
  bone2Name=dyn_suitSideFront_01_r
  swing1LimitDegrees=8
  swing2LimitDegrees=8
  twistLimitDegrees=45
  twistMotion=Locked
 }
 {
  bone1Name=base_suitSide_l
  bone2Name=spine_01
  swing1LimitDegrees=45
  swing1Motion=Locked
  swing2LimitDegrees=45
  swing2Motion=Locked
  twistLimitDegrees=45
  twistMotion=Locked
 }
 {bone1Name=dyn_suitSide_01_l, bone2Name=base_suitSide_l, swing1LimitDegrees=8, swing2LimitDegrees=2, twistLimitDegrees=45, twistMotion=Locked}
 {
  bone1Name=dyn_suitSide_02_l
  bone2Name=dyn_suitSide_01_l
  swing1LimitDegrees=8
  swing2LimitDegrees=2
  twistLimitDegrees=45
  twistMotion=Locked
 }
 {
  bone1Name=base_suitSide_r
  bone2Name=spine_01
  swing1LimitDegrees=45
  swing1Motion=Locked
  swing2LimitDegrees=45
  swing2Motion=Locked
  twistLimitDegrees=45
  twistMotion=Locked
 }
 {bone1Name=dyn_suitSide_01_r, bone2Name=base_suitSide_r, swing1LimitDegrees=8, swing2LimitDegrees=2, twistLimitDegrees=45, twistMotion=Locked}
 {
  bone1Name=dyn_suitSide_02_r
  bone2Name=dyn_suitSide_01_r
  swing1LimitDegrees=8
  swing2LimitDegrees=3
  twistLimitDegrees=45
  twistMotion=Locked
 }
 {bone1Name=thigh_l, bone2Name=pelvis, swing1LimitDegrees=45, swing2LimitDegrees=45, twistLimitDegrees=45}
 {bone1Name=thigh_r, bone2Name=pelvis, swing1LimitDegrees=45, swing2LimitDegrees=45, twistLimitDegrees=45}
 {
  bone1Name=ball_r
  bone2Name=foot_r
  swing1LimitDegrees=45
  swing1Motion=Free
  swing2LimitDegrees=45
  swing2Motion=Free
  twistLimitDegrees=45
  twistMotion=Free
 }
 {
  bone1Name=foot_r
  bone2Name=calf_r
  swing1LimitDegrees=45
  swing1Motion=Free
  swing2LimitDegrees=45
  swing2Motion=Free
  twistLimitDegrees=45
  twistMotion=Free
 }
 {
  bone1Name=calf_r
  bone2Name=thigh_r
  swing1LimitDegrees=45
  swing1Motion=Free
  swing2LimitDegrees=45
  swing2Motion=Free
  twistLimitDegrees=45
  twistMotion=Free
 }
 {
  bone1Name=calf_l
  bone2Name=thigh_l
  swing1LimitDegrees=45
  swing1Motion=Free
  swing2LimitDegrees=45
  swing2Motion=Free
  twistLimitDegrees=45
  twistMotion=Free
 }
 {
  bone1Name=ball_l
  bone2Name=foot_l
  swing1LimitDegrees=45
  swing1Motion=Free
  swing2LimitDegrees=45
  swing2Motion=Free
  twistLimitDegrees=45
  twistMotion=Free
 }
 {
  bone1Name=foot_l
  bone2Name=calf_l
  swing1LimitDegrees=45
  swing1Motion=Free
  swing2LimitDegrees=45
  swing2Motion=Free
  twistLimitDegrees=45
  twistMotion=Free
 }
 {
  bone1Name=base_suitBack
  bone2Name=pelvis
  swing1LimitDegrees=45
  swing1Motion=Locked
  swing2LimitDegrees=45
  swing2Motion=Locked
  twistLimitDegrees=45
  twistMotion=Locked
 }
 {
  bone1Name=base_suitSideBack_l
  bone2Name=pelvis
  swing1LimitDegrees=45
  swing1Motion=Locked
  swing2LimitDegrees=45
  swing2Motion=Locked
  twistLimitDegrees=45
  twistMotion=Locked
 }
 {
  bone1Name=base_suitSideBack_r
  bone2Name=pelvis
  swing1LimitDegrees=45
  swing1Motion=Locked
  swing2LimitDegrees=45
  swing2Motion=Locked
  twistLimitDegrees=45
  twistMotion=Locked
 }
 {
  bone1Name=base_suitSideFront_l
  bone2Name=pelvis
  swing1LimitDegrees=45
  swing1Motion=Locked
  swing2LimitDegrees=45
  swing2Motion=Locked
  twistLimitDegrees=45
  twistMotion=Locked
 }
 {
  bone1Name=base_suitSideFront_r
  bone2Name=pelvis
  swing1LimitDegrees=45
  swing1Motion=Locked
  swing2LimitDegrees=45
  swing2Motion=Locked
  twistLimitDegrees=45
  twistMotion=Locked
 }
 {
  bone1Name=base_suitSide_l
  bone2Name=pelvis
  swing1LimitDegrees=45
  swing1Motion=Locked
  swing2LimitDegrees=45
  swing2Motion=Locked
  twistLimitDegrees=45
  twistMotion=Locked
 }
 {
  bone1Name=base_suitSide_r
  bone2Name=pelvis
  swing1LimitDegrees=45
  swing1Motion=Locked
  swing2LimitDegrees=45
  swing2Motion=Locked
  twistLimitDegrees=45
  twistMotion=Locked
 }
 {
  bone1Name=deform_glute_l
  bone2Name=pelvis
  swing1LimitDegrees=45
  swing1Motion=Locked
  swing2LimitDegrees=45
  swing2Motion=Locked
  twistLimitDegrees=45
  twistMotion=Locked
 }
 {
  bone1Name=deform_glute_r
  bone2Name=pelvis
  swing1LimitDegrees=45
  swing1Motion=Locked
  swing2LimitDegrees=45
  swing2Motion=Locked
  twistLimitDegrees=45
  twistMotion=Locked
 }
]
