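# Animation Blueprint for the female GongFiddle body cosmetic, adding secondary motion to hood, cape and zippers over the parent mesh's copied pose.
# A Control Rig reacts to crouching and skydiving, feeding two RigidBody simulations selected by skydive state.
# Weighted constraints anchor zipper bases, twist bones are scaled, and AnimDynamics springs move the dangling parts.
# In the front end, an idle-bypass curve blends slot animation onto hair, hood and cape.
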AnimBlueprint /BRCosmetics/Characters/Player/Female/Medium/Bodies/F_MED_GongFiddle/Meshes/F_MED_GongFiddle_AnimBP
target_skeleton=&/F_MED_GongFiddle_Skeleton
graphs [
 AnimGraph {nodes [
   Constraint!(ComponentPose, Alpha=1, ConstraintWeights_0=0.171, ConstraintWeights_1:float=0.6798, ConstraintWeights_2:float=0.1492) -> (Pose) {
    constraintSetup [{targetBone {boneName=lowerarm_l}}, {targetBone {boneName=lowerarm_twist_01_l}}, {targetBone {boneName=lowerarm_twist_02_l}}]
    constraintWeights [1,1,1]
    boneToModify=~
   }
   n1 = Layered blend per bone(
    BasePose=Use cached pose 'StartAnim'()
    BlendPoses_0=Slot 'DefaultSlot'()
    BlendWeights_0=Get Custom Idle id Playing Fortnite() -> (Custom Idle id Playing Fortnite:double)
   ) {layerSetup [{branchFilters [
       {boneName=base_Hair_01_back}
       {boneName=Hood_AnimCore_SideR}
       {boneName=Hood_AnimCore_BackR}
       {boneName=Hood_AnimCore_BackL}
       {boneName=Hood_AnimCore_SideL}
       {boneName=Hood_AnimCore_FrL}
       {boneName=Cape_dyn}
      ]}]}
   Output Pose(Result=@n1) {layerGroup=None}
   n2 = Control Rig!(
    Source=Copy Pose From Mesh() {bUseAttachedParent=true}
    Alpha
    isCrouchingAndMoving:bool=Get isCrouchingAndMoving() -> (isCrouchingAndMoving:bool)
    isCrouchingNotMoving:bool=Get isCrouchingNotMoving() -> (isCrouchingNotMoving:bool)
    skygive:bool=Get bIsSkyDiving()
   ) -> (Pose) {controlRigAssetReference {blueprintRigClass=&/F_MED_GongFiddle_CtrlRig.F_MED_GongFiddle_CtrlRig_C}}
   Save cached pose 'CachposeK'(Pose=@n2) {cacheName=CachposeK}
   n3 = RigidBody!(ComponentPose=Local To Component(LocalPose=Use cached pose 'CachposeK'()), Alpha, OverrideWorldGravity:Vector="0,0,0", ExternalForce) -> (Pose) {
    bClampLinearTranslationLimitToRefPose=true
    bOverrideWorldGravity=true
    overridePhysicsAsset=&/F_MED_GongFiddle_Physics
    simulationSpace=BaseBoneSpace
    baseBoneRef {boneName=spine_01}
    componentAppliedLinearAccClamp=<3000 3000 3000>
    simSpaceSettings {
     dampingAlpha=0.1
     maxAngularAcceleration=500
     maxAngularVelocity=6
     maxLinearAcceleration=1200
     maxLinearVelocity=200
     velocityScaleZ=0.5
     worldAlpha=0.6
     externalLinearDragV=<0.3 0.3 1>
    }
   }
   n4 = Component To Local(ComponentPose=@n3)
   n5 = RigidBody!(
    ComponentPose=Local To Component(LocalPose=Use cached pose 'CachposeK'())
    Alpha
    OverrideWorldGravity="0,0,0"
    ExternalForce
    SimSpaceSettings:Sim Space Settings Structure=(WorldAlpha=0.6,VelocityScaleZ=0.5,DampingAlpha=0.1,MaxLinearVelocity=200,MaxAngularVelocity=6,MaxLinearAcceleration=400,MaxAngularAcceleration=20,ExternalLinearDragV=(X=0.3,Y=0.3,Z=1),ExternalLinearVelocity=(X=0,Y=0,Z=0),ExternalAngularVelocity=(X=0,Y=0,Z=0))
   ) -> (Pose) {
    bOverrideWorldGravity=true
    overridePhysicsAsset=&/F_MED_GongFiddle_Physics_SkyDive
    simulationSpace=BaseBoneSpace
    baseBoneRef {boneName=spine_01}
    componentAppliedLinearAccClamp=<2000 2000 2000>
    simSpaceSettings {
     dampingAlpha=0.1
     maxAngularAcceleration=20
     maxAngularVelocity=6
     maxLinearAcceleration=400
     maxLinearVelocity=200
     velocityScaleZ=0.5
     worldAlpha=0.6
     externalLinearDragV=<0.3 0.3 1>
    }
   }
   n6 = Component To Local(ComponentPose=@n5)
   n7 = Get bIsSkyDiving()
   n8 = Anim Dynamics - Bone: Cape_dyn(
    ComponentPose=Local To Component(LocalPose=Blend Poses by bool(BlendPose_0=@n6, BlendPose_1=@n4, bActiveValue=@n7) {blendType=Linear})
    Alpha=1
    ExternalForce="0,0,0"
   ) {
    boundBone {boneName=Cape_dyn}
    physicsBodyDefinitions [{
      boundBone {boneName=Cape_dyn}
      boxExtents=<10 10 10>
      constraintSetup {angularLimitsMax=<0 0 0>, angularLimitsMin=<0 0 0>, angularTarget=<0 0 0>}
      localJointOffset=<1 0 0>
     }]
   }
   n9 = "Constraint - Base_zip_l (PPPPPPPP)"!(
    ComponentPose=@n8
    Alpha=1
    ConstraintWeights_0=0.0905
    ConstraintWeights_1:float=0.0341
    ConstraintWeights_2:float=0.1634
    ConstraintWeights_3:float=0.0161
    ConstraintWeights_4:float=0.0983
    ConstraintWeights_5:float=0.2432
    ConstraintWeights_6:float=0.2913
    ConstraintWeights_7:float=0.0631
   ) -> (Pose) {
    boneToModify {boneName=Base_zip_l}
    constraintSetup [
     {targetBone {boneName=spine_04}}
     {targetBone {boneName=spine_03}}
     {targetBone {boneName=spine_02}}
     {targetBone {boneName=spine_01}}
     {targetBone {boneName=pelvis}}
     {targetBone {boneName=dyn_suitSide_01_l}}
     {targetBone {boneName=dyn_suitSideFront_01_l}}
     {targetBone {boneName=dyn_suitSideFront_02_l}}
    ]
    constraintWeights [1,1,1,1,1,1,1,1]
   }
   n10 = "Constraint - Base_zip_r (PPPPPPPP)"!(
    ComponentPose=@n9
    Alpha=1
    ConstraintWeights_0=0.0933
    ConstraintWeights_1:float=0.0238
    ConstraintWeights_2:float=0.1757
    ConstraintWeights_3:float=0.0429
    ConstraintWeights_4:float=0.3625
    ConstraintWeights_5:float=0.2365
    ConstraintWeights_6:float=0.0463
    ConstraintWeights_7:float=0.0189
   ) -> (Pose) {
    boneToModify {boneName=Base_zip_r}
    constraintSetup [
     {targetBone {boneName=spine_04}}
     {targetBone {boneName=spine_03}}
     {targetBone {boneName=spine_02}}
     {targetBone {boneName=pelvis}}
     {targetBone {boneName=dyn_suitSide_01_r}}
     {targetBone {boneName=dyn_suitSideFront_01_r}}
     {targetBone {boneName=dyn_suitSide_02_r}}
     {targetBone {boneName=dyn_suitSideFront_02_r}}
    ]
    constraintWeights [1,1,1,1,1,1,1,1]
   }
   n11 = "Constraint - Base_zip_fr (PPPPPPP)"!(
    ComponentPose=@n10
    Alpha=1
    ConstraintWeights_0=0
    ConstraintWeights_1:float=0
    ConstraintWeights_2:float=0
    ConstraintWeights_3:float=0
    ConstraintWeights_4:float=0.0507
    ConstraintWeights_5:float=0.0001
    ConstraintWeights_6:float=0.9492
   ) -> (Pose) {
    boneToModify {boneName=Base_zip_fr}
    constraintSetup [
     {targetBone {boneName=spine_04}}
     {targetBone {boneName=spine_03}}
     {targetBone {boneName=spine_02}}
     {targetBone {boneName=spine_01}}
     {targetBone {boneName=dyn_suitSideFront_01_r}}
     {targetBone {boneName=dyn_suitSide_02_r}}
     {targetBone {boneName=dyn_suitSideFront_02_r}}
    ]
    constraintWeights [1,1,1,1,1,1,1]
   }
   n12 = "Constraint - Base_zip_l_bot (PPPPPPPP)"!(
    ComponentPose=@n11
    Alpha=1
    ConstraintWeights_0=0.0022
    ConstraintWeights_1:float=0.0008
    ConstraintWeights_2:float=0.0046
    ConstraintWeights_3:float=0.0007
    ConstraintWeights_4:float=0.0188
    ConstraintWeights_5:float=0.1116
    ConstraintWeights_6:float=0.7837
    ConstraintWeights_7:float=0.0775
   ) -> (Pose) {
    boneToModify {boneName=Base_zip_l_bot}
    constraintSetup [
     {targetBone {boneName=spine_04}}
     {targetBone {boneName=spine_03}}
     {targetBone {boneName=spine_02}}
     {targetBone {boneName=pelvis}}
     {targetBone {boneName=dyn_suitSide_01_l}}
     {targetBone {boneName=dyn_suitSideFront_01_l}}
     {targetBone {boneName=dyn_suitSideFront_02_l}}
     {targetBone {boneName=dyn_suitSide_02_l}}
    ]
    constraintWeights [1,1,1,1,1,1,1,1]
   }
   n13 = "Constraint - Base_zip_l_hand (PPP)"!(ComponentPose=@n12, Alpha=1, ConstraintWeights_0=0.171, ConstraintWeights_1:float=0.6798, ConstraintWeights_2:float=0.1492) -> (Pose) {
    boneToModify {boneName=Base_zip_l_hand}
    constraintSetup [{targetBone {boneName=lowerarm_l}}, {targetBone {boneName=lowerarm_twist_01_l}}, {targetBone {boneName=lowerarm_twist_02_l}}]
    constraintWeights [1,1,1]
   }
   n14 = "Constraint - Base_zip_r_hand (PPP)"!(ComponentPose=@n13, Alpha=1, ConstraintWeights_0=0.6724, ConstraintWeights_1:float=0.1739, ConstraintWeights_2:float=0.1537) -> (Pose) {
    boneToModify {boneName=Base_zip_r_hand}
    constraintSetup [{targetBone {boneName=lowerarm_twist_01_r}}, {targetBone {boneName=lowerarm_r}}, {targetBone {boneName=lowerarm_twist_02_r}}]
    constraintWeights [1,1,1]
   }
   n15 = Anim Dynamics - Bone: dyn_zip_l(
    ComponentPose="Transform (Modify) Bone - Bone: upperarm_twist_02_l"!(
     ComponentPose="Transform (Modify) Bone - Bone: upperarm_twist_02_r"!(ComponentPose=@n14, bAlphaBoolEnabled:bool=True, Translation="0,0,0", Rotation="0,0,0", Scale="1,1.2,1.2") -> (Pose) {alphaInputType=Bool, scaleMode=BMM_Additive, boneToModify {boneName=upperarm_twist_02_r}, scale {y=1.2, z=1.2}}
     bAlphaBoolEnabled:bool=True
     Translation="0,0,0"
     Rotation="0,0,0"
     Scale="1,1.2,1.2"
    ) -> (Pose) {alphaInputType=Bool, scaleMode=BMM_Additive, boneToModify {boneName=upperarm_twist_02_l}, scale {y=1.2, z=1.2}}
    Alpha=1
    ExternalForce="0,0,0"
   ) {
    angularDampingOverride=1
    angularSpringConstant=100
    bAngularSpring=true
    bOverrideAngularDamping=true
    bUseGravityOverride=true
    bUsePlanarLimit=false
    gravityScale=0.35
    simulationSpace=World
    boundBone {boneName=dyn_zip_l}
    componentAppliedLinearAccClamp=<1000 1000 1000>
    componentLinearAccScale=<0.2 0.2 0.2>
    componentLinearVelScale=<0.2 0.2 0.2>
    gravityOverride {z=-980}
    physicsBodyDefinitions [{
      sphereCollisionRadius=1.3
      boundBone {boneName=dyn_zip_l}
      boxExtents=<10 10 10>
      constraintSetup {angularLimitsMax=<0 -2.5 10>, angularLimitsMin=<0 -60 -10>, angularTarget=<1 0 0>}
      localJointOffset=<0.4 0 0>
     }]
    planarLimits [{drivingBone {boneName=None}, planeTransform {location=<0 0 0>, rotation=<rot 0 0 0>}}]
    relativeSpaceBone {boneName=head}
   }
   n16 = Anim Dynamics - Bone: dyn_zip_r(ComponentPose=@n15, Alpha=1, ExternalForce="0,0,0") {
    angularDampingOverride=1
    angularSpringConstant=100
    bAngularSpring=true
    bOverrideAngularDamping=true
    bUseGravityOverride=true
    bUsePlanarLimit=false
    gravityScale=0.35
    simulationSpace=World
    boundBone {boneName=dyn_zip_r}
    componentAppliedLinearAccClamp=<1000 1000 1000>
    componentLinearAccScale=<0.2 0.2 0.2>
    componentLinearVelScale=<0.2 0.2 0.2>
    gravityOverride {z=-980}
    physicsBodyDefinitions [{
      sphereCollisionRadius=1.3
      boundBone {boneName=dyn_zip_r}
      boxExtents=<10 10 10>
      constraintSetup {angularLimitsMax=<0 -2.5 10>, angularLimitsMin=<0 -60 -10>, angularTarget=<1 0 0>}
      localJointOffset=<0.4 0 0>
     }]
    planarLimits [{drivingBone {boneName=None}, planeTransform {location=<0 0 0>, rotation=<rot 0 0 0>}}]
    relativeSpaceBone {boneName=head}
   }
   n17 = Anim Dynamics - Bone: dyn_zip_fr(ComponentPose=@n16, Alpha=1, ExternalForce="0,0,0") {
    angularDampingOverride=1
    angularSpringConstant=100
    bAngularSpring=true
    bOverrideAngularDamping=true
    bUseGravityOverride=true
    bUsePlanarLimit=false
    gravityScale=0.35
    simulationSpace=World
    boundBone {boneName=dyn_zip_fr}
    componentAppliedLinearAccClamp=<1000 10000 1000>
    componentLinearAccScale=<0.2 0.2 0.2>
    componentLinearVelScale=<0.2 0.2 0.2>
    gravityOverride {z=-980}
    physicsBodyDefinitions [{
      sphereCollisionRadius=1.3
      boundBone {boneName=dyn_zip_fr}
      boxExtents=<5 5 5>
      constraintSetup {angularLimitsMax=<0 1 10>, angularLimitsMin=<0 -45 -10>, angularTarget=<1 0 0>}
      localJointOffset=<0.4 0 0>
     }]
    planarLimits [{drivingBone {boneName=None}, planeTransform {location=<0 0 0>, rotation=<rot 0 0 0>}}]
    relativeSpaceBone {boneName=head}
   }
   n18 = Anim Dynamics - Bone: dyn_zip_l_bot(ComponentPose=@n17, Alpha=1, ExternalForce="0,0,0") {
    angularDampingOverride=1
    angularSpringConstant=100
    bAngularSpring=true
    bOverrideAngularDamping=true
    bUseGravityOverride=true
    bUsePlanarLimit=false
    gravityScale=0.35
    simulationSpace=World
    boundBone {boneName=dyn_zip_l_bot}
    componentAppliedLinearAccClamp=<500 500 500>
    componentLinearAccScale=<0.2 0.2 0.2>
    componentLinearVelScale=<0.2 0.2 0.2>
    gravityOverride {z=-980}
    physicsBodyDefinitions [{
      sphereCollisionRadius=1.3
      boundBone {boneName=dyn_zip_l_bot}
      boxExtents=<5 5 5>
      constraintSetup {angularLimitsMax=<0 1 20>, angularLimitsMin=<0 -40 -20>, angularTarget=<1 0 0>}
      localJointOffset=<0.6 0 0>
     }]
    planarLimits [{drivingBone {boneName=None}, planeTransform {location=<0 0 0>, rotation=<rot 0 0 0>}}]
    relativeSpaceBone {boneName=head}
   }
   n19 = Anim Dynamics - Bone: Hood_core_dyn(ComponentPose=@n18, Alpha=1, ExternalForce="0,0,0") {
    angularDampingOverride=1
    angularSpringConstant=400
    bAngularSpring=true
    bLinearSpring=true
    bOverrideAngularDamping=true
    bOverrideLinearDamping=true
    bUseGravityOverride=true
    linearDampingOverride=1
    linearSpringConstant=200
    simulationSpace=RootRelative
    alphaBoolBlend {blendOption=Linear}
    boundBone {boneName=Hood_core_dyn}
    componentAppliedLinearAccClamp=<400 400 400>
    componentLinearAccScale=<0.2 0.2 0.2>
    componentLinearVelScale=<0.2 0.2 0.2>
    gravityOverride {z=-980}
    physicsBodyDefinitions [{
      boundBone {boneName=Hood_core_dyn}
      boxExtents=<10 10 10>
      constraintSetup {angularLimitsMax=<0 0 0>, angularLimitsMin=<0 0 -30>, angularTarget=<1 0 0>, linearAxesMax=<0 0 1.5>, linearAxesMin=<-1 -1 -1.5>}
      localJointOffset=<1 0 0>
     }]
    relativeSpaceBone {boneName=spine_05}
   }
   n20 = Anim Dynamics - Bone: Cape_dyn(ComponentPose=@n19, Alpha=1, ExternalForce="0,0,0") {
    angularDampingOverride=1
    angularSpringConstant=200
    bAngularSpring=true
    bLinearSpring=true
    bOverrideAngularDamping=true
    bOverrideLinearDamping=true
    bUseGravityOverride=true
    linearDampingOverride=1
    linearSpringConstant=100
    simulationSpace=RootRelative
    alphaBoolBlend {blendOption=Linear}
    boundBone {boneName=Cape_dyn}
    componentAppliedLinearAccClamp=<1000 1000 1000>
    gravityOverride {z=-980}
    physicsBodyDefinitions [{
      boundBone {boneName=Cape_dyn}
      boxExtents=<12 12 12>
      constraintSetup {angularLimitsMax=<0 0 0>, angularLimitsMin=<0 0 0>, angularTarget=<0 0 0>, linearAxesMax=<0 0 1>, linearAxesMin=<0 -1 -1>}
      localJointOffset=<0 0 0>
     }]
    relativeSpaceBone {boneName=spine_05}
   }
   n21 = Anim Dynamics - Bone: dyn_zip_l_hand(ComponentPose=@n20, Alpha=1, ExternalForce="0,0,0") {
    angularDampingOverride=1
    angularSpringConstant=100
    bAngularSpring=true
    bOverrideAngularDamping=true
    bUseGravityOverride=true
    bUsePlanarLimit=false
    gravityScale=0.35
    simulationSpace=World
    boundBone {boneName=dyn_zip_l_hand}
    componentAppliedLinearAccClamp=<500 500 500>
    componentLinearAccScale=<0.2 0.2 0.2>
    componentLinearVelScale=<0.2 0.2 0.2>
    gravityOverride {z=-980}
    physicsBodyDefinitions [{
      sphereCollisionRadius=1.3
      boundBone {boneName=dyn_zip_l_hand}
      boxExtents=<5 5 5>
      constraintSetup {angularLimitsMax=<0 120 10>, angularLimitsMin=<0 -25 -10>, angularTarget=<1 0 0>}
      localJointOffset=<1 0 0>
     }]
    planarLimits [{drivingBone {boneName=None}, planeTransform {location=<0 0 0>, rotation=<rot 0 0 0>}}]
    relativeSpaceBone {boneName=head}
   }
   n22 = Anim Dynamics - Bone: dyn_zip_r_hand(ComponentPose=@n21, Alpha=1, ExternalForce="0,0,0") {
    angularDampingOverride=1
    angularSpringConstant=100
    bAngularSpring=true
    bOverrideAngularDamping=true
    bUseGravityOverride=true
    bUsePlanarLimit=false
    gravityScale=0.35
    simulationSpace=World
    boundBone {boneName=dyn_zip_r_hand}
    componentAppliedLinearAccClamp=<500 500 500>
    componentLinearAccScale=<0.2 0.2 0.2>
    componentLinearVelScale=<0.2 0.2 0.2>
    gravityOverride {z=-980}
    physicsBodyDefinitions [{
      sphereCollisionRadius=1.3
      boundBone {boneName=dyn_zip_r_hand}
      boxExtents=<5 5 5>
      constraintSetup {angularLimitsMax=<0 25 10>, angularLimitsMin=<0 -100 -10>, angularTarget=<1 0 0>}
      localJointOffset=<1 0 0>
     }]
    planarLimits [{drivingBone {boneName=None}, planeTransform {location=<0 0 0>, rotation=<rot 0 0 0>}}]
    relativeSpaceBone {boneName=head}
   }
   Save cached pose 'StartAnim'(Pose=Component To Local(ComponentPose=@n22)) {cacheName=StartAnim}
  ]}
 EventGraph {nodes [
   Event BlueprintUpdateAnimation()
   > n1 = "Gravity Override State Function (Parameter Asset) (Nativized)"(
    CharacterPartAnimInstance=Self-Reference()
    GameTimeInSeconds=GetGameTimeInSeconds()
    DebugJointName=spine_05
    Parameters=&/M_MED_GongFiddle_Parameters
   )
   > Set Gravity Override(Gravity Override=@n1)
   > Set RBAN Settings Out(RBAN Settings Out=@n1.OutRigidBodyAnimNodeParameters)
   n2 = Get bIsMoving2D()
   n3 = Get bIsCrouching()
   > Set isCrouchingNotMoving(isCrouchingNotMoving:bool=AND Boolean(A=@n3, B=NOT Boolean(A=@n2))) -> (Output_Get:bool)
   > Set isCrouchingAndMoving(isCrouchingAndMoving:bool=AND Boolean(A=@n3, B=@n2)) -> (Output_Get:bool)
   > Branch(Condition=Get bIsInFrontEnd())
   > Set Custom Idle id Playing Fortnite(Custom Idle id Playing Fortnite:double=GetCurveValue(self=GetOwnerFrontEndAnimInstance(), CurveName=Frontend_Idle_Bypass)) -> (Output_Get:double)
  ]}
]
variables [
 Custom Idle id Playing Fortnite {type=number}
 Gravity Override {type=Vector}
 isCrouchingAndMoving {type=boolean}
 isCrouchingNotMoving {type=boolean}
 RBAN Settings Out {type=FortRigidBodyAnimNodeParameters}
]
