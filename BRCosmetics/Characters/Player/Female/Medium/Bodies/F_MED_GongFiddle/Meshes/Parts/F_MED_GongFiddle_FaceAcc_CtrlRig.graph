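# Control Rig for the GongFiddle female face accessory mesh, correcting hair base bone transforms by pose and state.
# A right upper-arm spherical pose reader nudges the right hair bones, then fixed offsets apply to the hair bases.
# SkyDive selects a posed set; otherwise Emote applies offsets weighted by the product of both arm pose readers and remapped head-to-spine rotation.
# A Cape_dyn-to-Hair_22 parent constraint and some other nodes are left unconnected.
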ControlRigBlueprint /BRCosmetics/Characters/Player/Female/Medium/Bodies/F_MED_GongFiddle/Meshes/Parts/F_MED_GongFiddle_FaceAcc_CtrlRig
graphs [
 Multiply_1_ContainedGraph {nodes [
   n1 = RigVMFunctionEntryNode()
   Multiply = &RigVMFunction_MathDoubleMul(A=@n1.A, B=@n1.B)
   RigVMFunctionReturnNode(Result=&RigVMFunction_MathDoubleMul(B=@n1.C, A=@Multiply.Result).Result)
  ]}
 RigVMFunctionLibrary {links [], nodes []}
 RigVMModel {nodes [
   &RigUnit_ModifyTransforms(
    ExecutePin=()
    ItemToModify="((Item=(Type=Bone,Name=\"base_Hair_01_back\"),Transform=(Rotation=(X=0,Y=-0.3907,Z=0,W=0.9205),Translation=(X=-2,Y=0,Z=2),Scale3D=(X=1,Y=1,Z=1))),(Item=(Type=Bone,Name=\"base_Hair_01_r_back\"),Transform=(Rotation=(X=0,Y=0,Z=0.1564,W=0.9877),Translation=(X=-1.7 ...(truncated)"
    Weight=1
    WeightMinimum=0
   )
   > &RigUnit_ModifyTransforms(
    ItemToModify="((Item=(Type=Bone,Name=\"base_Hair_01_l\"),Transform=(Rotation=(X=0,Y=0,Z=0,W=1),Translation=(X=0,Y=0,Z=0),Scale3D=(X=1,Y=1,Z=1))),(Item=(Type=Bone,Name=\"base_Hair_01_r\"),Transform=(Rotation=(X=0,Y=0,Z=0,W=1),Translation=(X=-1,Y=0,Z=0),Scale3D=(X=1,Y=1,Z=1)) ...(truncated)"
    Weight=1
    WeightMinimum=0
   )
   RigUnit_BeginExecution = &RigUnit_BeginExecution()
   RigVMFunction_Sequence = &RigVMFunction_Sequence(ExecuteContext=@RigUnit_BeginExecution.ExecutePin)
   n1 = &RigUnit_SphericalPoseReader(
    ActiveRegionScaleFactors=(PositiveWidth=1,NegativeWidth=1,PositiveHeight=0.9,NegativeHeight=1)
    ActiveRegionSize=0.12
    Debug=(bDrawDebug=false,bDraw2D=False,bDrawLocalAxes=False,DebugScale=25,DebugSegments=20,DebugThickness=0.25)
    DriverAxis=(X=-1,Y=0,Z=0)
    DriverItem=(Type=Bone,Name="upperarm_r")
    FalloffRegionScaleFactors=(PositiveWidth=1,NegativeWidth=1,PositiveHeight=1,NegativeHeight=1)
    FalloffSize=0.26
    FlipHeightScaling=False
    FlipWidthScaling=false
    OptionalParentItem=(Type=Bone,Name="None")
    RotationOffset=(X=-125.27,Y=28.964,Z=17.3392)
    ExecutePin=@RigVMFunction_Sequence.A
   )
   > &RigUnit_ModifyTransforms(
    ItemToModify=((Item=(Type=Bone,Name="base_Hair_01_r"),Transform=(Rotation=(X=0,Y=0,Z=0,W=1),Translation=(X=-2,Y=0,Z=-1),Scale3D=(X=1,Y=1,Z=1))))
    WeightMinimum=0
    Weight=@n1.OutputParam
   )
   > &RigUnit_ModifyTransforms(
    ItemToModify=((Item=(Type=Bone,Name="base_Hair_01_r_back"),Transform=(Rotation=(X=0,Y=0,Z=0,W=1),Translation=(X=-2,Y=0,Z=0),Scale3D=(X=1,Y=1,Z=1))))
    WeightMinimum=0
    Weight=@n1.OutputParam
   )
   &RigUnit_ModifyTransforms(
    ItemToModify=((Item=(Type=Bone,Name="base_Hair_01_r"),Transform=(Rotation=(X=0,Y=0,Z=0,W=1),Translation=(X=0,Y=0,Z=0),Scale3D=(X=1,Y=1,Z=1))))
    Weight=1
    WeightMinimum=0
    ExecutePin=@RigVMFunction_Sequence.B
   )
   > &RigUnit_ModifyTransforms(
    ItemToModify=((Item=(Type=Bone,Name="base_Hair_01_r_back"),Transform=(Rotation=(X=0,Y=0,Z=0,W=1),Translation=(X=-0.5,Y=0,Z=0),Scale3D=(X=1,Y=1,Z=1))))
    Weight=1
    WeightMinimum=0
   )
   > &RigUnit_ModifyTransforms(
    ItemToModify=((Item=(Type=Bone,Name="base_Hair_01_l_back"),Transform=(Rotation=(X=0,Y=0,Z=0,W=1),Translation=(X=-1,Y=0,Z=0),Scale3D=(X=1,Y=1,Z=1))))
    Weight=1
    WeightMinimum=0
   )
   > &RigUnit_ModifyTransforms(
    ItemToModify=((Item=(Type=Bone,Name="base_Hair_01_l"),Transform=(Rotation=(X=0,Y=0,Z=0,W=1),Translation=(X=-1,Y=0,Z=0),Scale3D=(X=1,Y=1,Z=1))))
    Weight=1
    WeightMinimum=0
   )
   > n2 = &RigUnit_ModifyTransforms(
    ItemToModify=((Item=(Type=Bone,Name="base_Hair_01_back"),Transform=(Rotation=(X=0,Y=0,Z=0,W=1),Translation=(X=-1,Y=0,Z=0),Scale3D=(X=1,Y=1,Z=1))))
    Weight=1
    WeightMinimum=0
   )
   RigVMFunction_ControlFlowBranch = &RigVMFunction_ControlFlowBranch(ExecuteContext=@n2.ExecutePin, Condition=RigVMVariableNode() {variable=SkyDive}.Value)
   n3 = &RigVMFunction_ControlFlowBranch(ExecuteContext=@RigVMFunction_ControlFlowBranch.False, Condition=RigVMVariableNode() {variable=Emote}.Value)
   n4 = &RigUnit_SphericalPoseReader(
    ActiveRegionScaleFactors=(PositiveWidth=1,NegativeWidth=1,PositiveHeight=0.9,NegativeHeight=1)
    ActiveRegionSize=0.15
    Debug=(bDrawDebug=true,bDraw2D=False,bDrawLocalAxes=False,DebugScale=25,DebugSegments=20,DebugThickness=0.25)
    DriverAxis=(X=-1,Y=0,Z=0)
    DriverItem=(Type=Bone,Name="upperarm_r")
    FalloffRegionScaleFactors=(PositiveWidth=1,NegativeWidth=1,PositiveHeight=1,NegativeHeight=1)
    FalloffSize=0.26
    FlipHeightScaling=False
    FlipWidthScaling=false
    OptionalParentItem=(Type=Bone,Name="None")
    RotationOffset=(X=-152.8852,Y=-52.4613,Z=-15.69)
    ExecutePin=@n3.True
   )
   > n5 = &RigUnit_SphericalPoseReader(
    ActiveRegionScaleFactors=(PositiveWidth=1,NegativeWidth=1,PositiveHeight=0.9,NegativeHeight=1)
    ActiveRegionSize=0.15
    Debug=(bDrawDebug=true,bDraw2D=False,bDrawLocalAxes=False,DebugScale=25,DebugSegments=20,DebugThickness=0.25)
    DriverAxis=(X=1,Y=0,Z=0)
    DriverItem=(Type=Bone,Name="upperarm_l")
    FalloffRegionScaleFactors=(PositiveWidth=1,NegativeWidth=1,PositiveHeight=1,NegativeHeight=1)
    FalloffSize=0.26
    FlipHeightScaling=False
    FlipWidthScaling=false
    OptionalParentItem=(Type=Bone,Name="None")
    RotationOffset=(X=1.4337,Y=-52.6221,Z=-19.4231)
   )
   n6 = &RigUnit_GetRelativeTransformForItem(Child=(Type=Bone,Name="head"), Parent=(Type=Bone,Name="spine_05"), bChildInitial=False, bParentInitial=False)
   Multiply = RigVMAggregateNode(
    C=&RigVMFunction_MathDoubleRemap(
     SourceMaximum=-30
     SourceMinimum=0
     TargetMaximum=1
     TargetMinimum=0
     Value=&RigVMFunction_MathQuaternionToEuler(RotationOrder=ZYX, Value=@n6.RelativeTransform.Rotation).Result.Z
    ).Result
    B=@n4.OutputParam
    A=@n5.OutputParam
   )
   > &RigUnit_ModifyTransforms(
    ItemToModify="((Item=(Type=Bone,Name=\"base_Hair_01_back\"),Transform=(Rotation=(X=0,Y=0,Z=0,W=1),Translation=(X=-4.0376,Y=0,Z=1.7451),Scale3D=(X=1,Y=1,Z=1))),(Item=(Type=Bone,Name=\"base_Hair_01_r_back\"),Transform=(Rotation=(X=0,Y=0,Z=0,W=1),Translation=(X=-6.7002,Y=0,Z=- ...(truncated)"
    WeightMinimum=0
    Weight=@Multiply.Result
   )
   > &RigUnit_ModifyTransforms(
    ItemToModify="((Item=(Type=Bone,Name=\"base_Hair_01_l\"),Transform=(Rotation=(X=0,Y=0,Z=0,W=1),Translation=(X=0,Y=0,Z=0),Scale3D=(X=1,Y=1,Z=1))),(Item=(Type=Bone,Name=\"base_Hair_01_r\"),Transform=(Rotation=(X=0,Y=0,Z=0,W=1),Translation=(X=-1,Y=0,Z=0),Scale3D=(X=1,Y=1,Z=1)) ...(truncated)"
    WeightMinimum=0
    Weight=@Multiply.Result
   )
   &RigUnit_ModifyTransforms(
    ItemToModify="((Item=(Type=Bone,Name=\"base_Hair_01_back\"),Transform=(Rotation=(X=0,Y=-0.1816,Z=0,W=0.9834),Translation=(X=0,Y=0,Z=0),Scale3D=(X=1,Y=1,Z=1))),(Item=(Type=Bone,Name=\"base_Hair_01_r_back\"),Transform=(Rotation=(X=0,Y=0.2588,Z=0,W=0.9659),Translation=(X=-2,Y= ...(truncated)"
    Weight=1
    WeightMinimum=0
    ExecutePin=@RigVMFunction_ControlFlowBranch.True
   )
   > &RigUnit_ModifyTransforms(
    ItemToModify="((Item=(Type=Bone,Name=\"base_Hair_01_l\"),Transform=(Rotation=(X=0,Y=0.2588,Z=0,W=0.9659),Translation=(X=0,Y=0,Z=0),Scale3D=(X=1,Y=1,Z=1))),(Item=(Type=Bone,Name=\"base_Hair_01_r\"),Transform=(Rotation=(X=0,Y=0.2588,Z=0,W=0.9659),Translation=(X=0,Y=0,Z=0),Sca ...(truncated)"
    Weight=1
    WeightMinimum=0
   )
   &RigUnit_ParentConstraint(
    AdvancedSettings=(InterpolationType=Average,RotationOrderForFilter=XZY)
    Child=(Type=Bone,Name="Hair_22")
    ExecutePin=()
    Filter=(TranslationFilter=(bX=false,bY=false,bZ=false),RotationFilter=(bX=false,bY=false,bZ=true),ScaleFilter=(bX=True,bY=True,bZ=True))
    Parents=((Item=(Type=Bone,Name="Cape_dyn"),Weight=0.5))
    Weight=1
   )
   &RigVMFunction_MathDoubleRemap(
    SourceMaximum=-20
    SourceMinimum=0
    TargetMaximum=1
    TargetMinimum=0
    Value=&RigVMFunction_MathQuaternionToEuler(
     RotationOrder=ZYX
     Value=&RigUnit_GetRelativeTransformForItem(Child=(Type=Bone,Name="head"), Parent=(Type=Bone,Name="spine_05"), bChildInitial=False, bParentInitial=False).RelativeTransform.Rotation
    ).Result.Z
   )
  ]}
]
hierarchy [
 root {loc=<0 0 0>, rot=<0 0 0 1>}
 pelvis {parent=root, loc=<0 -0.5301 98.0571>, rot=<0.0127 0.707 -0.0127 -0.707>}
 spine_01 {parent=pelvis, loc=<6.1208 -0.9328 0>, rot=<0 0 0.034 -0.9994>}
 spine_02 {parent=spine_01, loc=<6.5997 -0.2095 0>, rot=<0 0 0.0238 -0.9997>}
 spine_03 {parent=spine_02, loc=<6.4914 -0.1237 0>, rot=<0 0 0.0149 -0.9999>}
 spine_04 {parent=spine_03, loc=<6.9598 0.0794 0>, rot=<0 0 -0.0132 -0.9999>}
 spine_05 {parent=spine_04, loc=<6.9987 0 0>, rot=<0 0 -0.0015 -1>}
 neck_01 {parent=spine_05, loc=<11.3138 0.1529 0>, rot=<0 0 -0.2003 -0.9797>}
 neck_02 {parent=neck_01, loc=<4.5627 0 0>, rot=<0 0 0.0504 -0.9987>}
 head {parent=neck_02, loc=<5.3914 0 0>, rot=<0 0 0.115 -0.9934>}
 Hair_20 {parent=head, loc=<6.7389 4.777 -7.1918>, rot=<0.1628 -0.9836 -0.0723 -0.0288>}
 Hair_21 {parent=Hair_20, loc=<5.8895 0 0>, rot=<0 0 0 1>}
 Hair_18 {parent=head, loc=<14.6275 3.4334 -4.2437>, rot=<0.2515 -0.9249 0.023 -0.2841>}
 Hair_19 {parent=Hair_18, loc=<10.2134 0 0>, rot=<0 0 0 1>}
 Hair_17 {parent=head, loc=<9.4133 4.3434 -7.844>, rot=<0.3687 -0.9166 -0.0576 -0.1433>}
 Hair_14 {parent=head, loc=<12.4306 8.8681 -0.1289>, rot=<-0.1788 0.7944 -0.1114 -0.5697>}
 Hair_15 {parent=Hair_14, loc=<6.5657 0 0>, rot=<0.0679 0.4626 -0.1814 -0.8652>}
 Hair_16 {parent=Hair_15, loc=<8.5817 0 0>, rot=<0 0 0 1>}
 Hair_12 {parent=head, loc=<13.7219 -3.7058 6.7905>, rot=<0.2792 -0.2808 0.9183 -0.004>}
 Hair_13 {parent=Hair_12, loc=<6.2721 0 0>, rot=<0 0 0 1>}
 base_Hair_01_r {parent=head, loc=<3.9486 2.202 8.1485>, rot=<0.0872 0.0004 0.9962 -0.0041>}
 dyn_Hair_01_r {parent=base_Hair_01_r, loc=<0 0 0>, rot=<0 0 0 1>}
 dyn_Hair_02_r {parent=dyn_Hair_01_r, loc=<6.1439 0 0>, rot=<0 0 0 1>}
 AnimCore_05_R {parent=dyn_Hair_02_r, loc=<4.6494 -0.3689 0.4171>, rot=<0 0 0 1>}
 AnimCore_07_R {parent=dyn_Hair_02_r, loc=<4.0948 3.5476 3.2007>, rot=<0 0 0 1>}
 base_Hair_01_r_back {parent=head, loc=<4.9422 -6.2239 8.4044>, rot=<0.0828 -0.2576 0.9624 -0.0259>}
 dyn_Hair_01_r_back {parent=base_Hair_01_r_back, loc=<0 0 0>, rot=<0 0 0 1>}
 dyn_Hair_02_r_back {parent=dyn_Hair_01_r_back, loc=<6.1439 0 0>, rot=<0 0 0 1>}
 AnimCore_06_R {parent=dyn_Hair_02_r_back, loc=<3.4102 -2.9863 -2.3028>, rot=<0 0 0 1>}
 AnimCore_08_R {parent=dyn_Hair_02_r_back, loc=<3.4102 -2.9863 1.8163>, rot=<0 0 0 1>}
 AnimCore_09_R {parent=dyn_Hair_02_r_back, loc=<3.4102 2.2142 1.2421>, rot=<0.118 0 0 -0.993>}
 base_Hair_01_l_back {parent=head, loc=<5.5088 -9.9769 -4.2841>, rot=<0.0584 -0.8641 0.4989 -0.0337>}
 dyn_Hair_01_l_back {parent=base_Hair_01_l_back, loc=<0 0 0>, rot=<0 0 0 1>}
 dyn_Hair_02_l_back {parent=dyn_Hair_01_l_back, loc=<6.1439 0 0>, rot=<0 0 0 1>}
 Hair_11 {parent=head, loc=<1.3223 -13.2513 4.3363>, rot=<0.0789 -0.5318 0.8411 -0.0598>}
 Hair_10 {parent=head, loc=<0.8373 -12.1718 0>, rot=<0.0029 -0.7071 0.7071 -0.0029>}
 Hair_09 {parent=head, loc=<-0.0165 -10.0786 -5.8538>, rot=<0.0037 -0.9027 0.4303 -0.0018>}
 Hair_08 {parent=head, loc=<-0.4777 -6.2022 10.2958>, rot=<0.1642 -0.1823 0.9694 0.001>}
 Hair_07 {parent=head, loc=<1.2966 -0.7514 12.0019>, rot=<0.5143 -0.0116 0.8355 0.193>}
 Hair_05 {parent=head, loc=<7.4126 2.7059 9.4216>, rot=<0.4853 -0.8534 0.094 0.1653>}
 Hair_06 {parent=Hair_05, loc=<4.1535 0 0>, rot=<0 0 0 1>}
 Hair_03 {parent=head, loc=<4.8299 8.3735 6.2631>, rot=<0.0549 0.5967 0.7996 0.0395>}
 Hair_04 {parent=Hair_03, loc=<3.6744 0 0>, rot=<0 0 0 1>}
 Hair_02 {parent=head, loc=<-0.9978 3.6296 8.962>, rot=<0.0041 -1 0 0>}
 Hair_01 {parent=head, loc=<-0.9495 -2.2083 -7.7715>, rot=<0.0041 -1 0 0>}
 Hair_22 {parent=head, loc=<0.1921 -11.9466 3.6987>, rot=<0.0789 -0.5318 0.8411 -0.0598>}
 base_Hair_01_l {parent=head, loc=<4.0017 -4.2151 -8.0783>, rot=<0.0041 -1 0 0>}
 dyn_Hair_01_l {parent=base_Hair_01_l, loc=<0 0 0>, rot=<0 0 0 1>}
 dyn_Hair_02_l {parent=dyn_Hair_01_l, loc=<6.1439 0 0>, rot=<0 0 0 1>}
 base_Hair_01_back {parent=head, loc=<5.5088 -9.9769 1.9975>, rot=<0.0476 -0.7055 0.7055 -0.0476>}
 dyn_Hair_01_back {parent=base_Hair_01_back, loc=<0 0 0>, rot=<0 0 0 1>}
 dyn_Hair_02_back {parent=dyn_Hair_01_back, loc=<6.1439 0 0>, rot=<0 0 0 1>}
 AnimCore_01_R {parent=dyn_Hair_02_back, loc=<3.1524 -5.1931 -5.0444>, rot=<-0.4945 0.0185 0.0737 -0.8658>}
 AnimCore_02_R {parent=dyn_Hair_02_back, loc=<2.7108 -1.0985 -3.9905>, rot=<0 0 0 1>}
 AnimCore_03_R {parent=dyn_Hair_02_back, loc=<2.7108 3.7401 -3.9905>, rot=<0 0 0 1>}
 Cape_dyn {parent=spine_05, loc=<10.7769 -10.25 0>, rot=<0 0 -0.2045 -0.9789>}
 clavicle_l {parent=spine_05, loc=<5.4831 -0.1053 -3.2946>, rot=<0.0465 -0.7289 -0.009 -0.6829>}
 upperarm_l {parent=clavicle_l, loc=<9.8658 -1.2136 -0.4105>, rot=<0.0164 -0.3245 -0.0269 -0.9454>}
 lowerarm_l {parent=upperarm_l, loc=<26.465 -0.8345 0.259>, rot=<0.0036 0.0161 -0.2298 -0.9731>}
 clavicle_r {parent=spine_05, loc=<5.4831 -0.1053 3.2946>, rot=<0.7289 0.0465 -0.6829 0.009>}
 upperarm_r {parent=clavicle_r, loc=<-9.8658 1.2136 0.4105>, rot=<0.0164 -0.3245 -0.0269 -0.9454>}
 lowerarm_r {parent=upperarm_r, loc=<-26.465 0.8345 -0.259>, rot=<0.0036 0.0161 -0.2298 -0.9731>}
 Hair_dynamics_ctrl {type=Curve, loc=<0 0 0>, rot=<0 0 0 1>}
]
variables [
 Emote {type=bool, default=False}
 SkyDive {type=bool, default=False}
]
