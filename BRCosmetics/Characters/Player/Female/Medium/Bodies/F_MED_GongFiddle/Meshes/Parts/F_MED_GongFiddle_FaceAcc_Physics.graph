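# Physics asset for the face accessory mesh of the female medium GongFiddle player body, setting up capsule bodies for secondary hair motion.
# Five two-segment hair chains (left, right and three back) swing within narrow limits from base bones angularly locked to the head.
# Angularly free constraints tie the side-back tips to the center-back tip.
# Head, neck, upper spine, clavicle and Cape_dyn capsules are angularly locked together, forming a rigid body proxy.
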PhysicsAsset /BRCosmetics/Characters/Player/Female/Medium/Bodies/F_MED_GongFiddle/Meshes/Parts/F_MED_GongFiddle_FaceAcc_Physics
bodies {
 Cape_dyn [{shapeType=Capsule, length=4.6196, radius=5.4881, shapeName=Cape_dyn_capsule, center=<0.3622 4.2792 0>}]
 base_Hair_01_back [{shapeType=Capsule, length=1, radius=1}]
 base_Hair_01_l [{shapeType=Capsule, length=1, radius=1}]
 base_Hair_01_l_back [{shapeType=Capsule, length=1, radius=1}]
 base_Hair_01_r [{shapeType=Capsule, length=1, radius=1}]
 base_Hair_01_r_back [{shapeType=Capsule, length=1, radius=1}]
 clavicle_l [{shapeType=Capsule, length=6.7344, radius=3.5118, shapeName=clavicle_l_capsule, center=<4.3669 0 0>, rotation=<rot 90 180 0>}]
 clavicle_r [{shapeType=Capsule, length=6.7344, radius=3.5118, shapeName=clavicle_r_capsule, center=<-4.3669 0 0>, rotation=<rot 90 0 0>}]
 dyn_Hair_01_back [{shapeType=Capsule, length=6.314, radius=3.8413, center=<1 0 2.7956>, rotation=<rot 0 0 90>}]
 dyn_Hair_01_l [{shapeType=Capsule, length=6.314, radius=3.8413, center=<1 0 3>, rotation=<rot 0 0 90>}]
 dyn_Hair_01_l_back [{shapeType=Capsule, length=6.314, radius=3.8413, center=<1 0 3>, rotation=<rot 0 0 90>}]
 dyn_Hair_01_r [{shapeType=Capsule, length=6.9144, radius=3.2409, center=<1 0 1>, rotation=<rot 0 0 90>}]
 dyn_Hair_01_r_back [{shapeType=Capsule, length=6.314, radius=3.8413, center=<1 0 3>, rotation=<rot 0 0 90>}]
 dyn_Hair_02_back [{shapeType=Capsule, length=6.0304, radius=4.1249, center=<3 0 2>, rotation=<rot -0.0035 0 90>}]
 dyn_Hair_02_l [{shapeType=Capsule, length=6.314, radius=3.8413, center=<1 0 3>, rotation=<rot 0 0 90>}]
 dyn_Hair_02_l_back [{shapeType=Capsule, length=6.314, radius=3.8413, center=<1 0 3>, rotation=<rot -0.003 0.0017 90>}]
 dyn_Hair_02_r [{shapeType=Capsule, length=6.1413, radius=4.0139, center=<1 0 0>, rotation=<rot 0 0 90>}]
 dyn_Hair_02_r_back [{shapeType=Capsule, length=4.9554, radius=3.2554, center=<1 0 1>, rotation=<rot -0.0017 -0.003 90.0005>}]
 head [{shapeType=Capsule, length=6.06, radius=6, center=<2.7089 0 -1.0016>, rotation=<rot 90 0 0>}]
 neck_02 [{shapeType=Capsule, length=5.5638, radius=3.451, rotation=<rot 90 0 0>}]
 spine_05 [{shapeType=Capsule, length=16.5779, radius=12.7784, shapeName=spine_05_capsule, center=<2.1015 -1.7289 0>, rotation=<rot 0 -90.0035 0>}]
}
constraints [
 {bone1Name=dyn_Hair_01_back, bone2Name=base_Hair_01_back, swing1LimitDegrees=5, swing2LimitDegrees=15, twistLimitDegrees=5}
 {bone1Name=dyn_Hair_02_back, bone2Name=dyn_Hair_01_back, swing1LimitDegrees=10, swing2LimitDegrees=20, twistLimitDegrees=5}
 {bone1Name=dyn_Hair_01_l, bone2Name=base_Hair_01_l, swing1LimitDegrees=4, swing2LimitDegrees=10, twistLimitDegrees=5}
 {bone1Name=dyn_Hair_02_l, bone2Name=dyn_Hair_01_l, swing1LimitDegrees=5, swing2LimitDegrees=4, twistLimitDegrees=5}
 {bone1Name=dyn_Hair_01_l_back, bone2Name=base_Hair_01_l_back, swing1LimitDegrees=7, swing2LimitDegrees=15, twistLimitDegrees=5}
 {bone1Name=dyn_Hair_02_l_back, bone2Name=dyn_Hair_01_l_back, swing1LimitDegrees=10, swing2LimitDegrees=20, twistLimitDegrees=5}
 {bone1Name=dyn_Hair_01_r, bone2Name=base_Hair_01_r, swing1LimitDegrees=10, swing2LimitDegrees=12, twistLimitDegrees=5}
 {bone1Name=dyn_Hair_02_r, bone2Name=dyn_Hair_01_r, swing1LimitDegrees=10, swing2LimitDegrees=12, twistLimitDegrees=5}
 {bone1Name=dyn_Hair_01_r_back, bone2Name=base_Hair_01_r_back, swing1LimitDegrees=7, swing2LimitDegrees=15, twistLimitDegrees=5}
 {bone1Name=dyn_Hair_02_r_back, bone2Name=dyn_Hair_01_r_back, swing1LimitDegrees=10, swing2LimitDegrees=20, twistLimitDegrees=5}
 {
  bone1Name=base_Hair_01_l
  bone2Name=head
  swing1LimitDegrees=45
  swing1Motion=Locked
  swing2LimitDegrees=45
  swing2Motion=Locked
  twistLimitDegrees=45
  twistMotion=Locked
 }
 {
  bone1Name=base_Hair_01_r
  bone2Name=head
  swing1LimitDegrees=45
  swing1Motion=Locked
  swing2LimitDegrees=45
  swing2Motion=Locked
  twistLimitDegrees=45
  twistMotion=Locked
 }
 {
  bone1Name=base_Hair_01_back
  bone2Name=head
  swing1LimitDegrees=45
  swing1Motion=Locked
  swing2LimitDegrees=45
  swing2Motion=Locked
  twistLimitDegrees=45
  twistMotion=Locked
 }
 {
  bone1Name=base_Hair_01_r_back
  bone2Name=head
  swing1LimitDegrees=45
  swing1Motion=Locked
  swing2LimitDegrees=45
  swing2Motion=Locked
  twistLimitDegrees=45
  twistMotion=Locked
 }
 {
  bone1Name=base_Hair_01_l_back
  bone2Name=head
  swing1LimitDegrees=45
  swing1Motion=Locked
  swing2LimitDegrees=45
  swing2Motion=Locked
  twistLimitDegrees=45
  twistMotion=Locked
 }
 {
  bone1Name=head
  bone2Name=neck_02
  swing1LimitDegrees=45
  swing1Motion=Locked
  swing2LimitDegrees=45
  swing2Motion=Locked
  twistLimitDegrees=45
  twistMotion=Locked
 }
 {
  bone1Name=dyn_Hair_02_r_back
  bone2Name=dyn_Hair_02_back
  swing1LimitDegrees=45
  swing1Motion=Free
  swing2LimitDegrees=45
  swing2Motion=Free
  twistLimitDegrees=45
  twistMotion=Free
 }
 {
  bone1Name=dyn_Hair_02_l_back
  bone2Name=dyn_Hair_02_back
  swing1LimitDegrees=45
  swing1Motion=Free
  swing2LimitDegrees=45
  swing2Motion=Free
  twistLimitDegrees=45
  twistMotion=Free
 }
 {
  bone1Name=neck_02
  bone2Name=spine_05
  swing1LimitDegrees=45
  swing1Motion=Locked
  swing2LimitDegrees=45
  swing2Motion=Locked
  twistLimitDegrees=45
  twistMotion=Locked
 }
 {
  bone1Name=clavicle_l
  bone2Name=spine_05
  swing1LimitDegrees=45
  swing1Motion=Locked
  swing2LimitDegrees=45
  swing2Motion=Locked
  twistLimitDegrees=45
  twistMotion=Locked
 }
 {
  bone1Name=clavicle_r
  bone2Name=spine_05
  swing1LimitDegrees=45
  swing1Motion=Locked
  swing2LimitDegrees=45
  swing2Motion=Locked
  twistLimitDegrees=45
  twistMotion=Locked
 }
 {
  bone1Name=Cape_dyn
  bone2Name=spine_05
  swing1LimitDegrees=45
  swing1Motion=Locked
  swing2LimitDegrees=45
  swing2Motion=Locked
  twistLimitDegrees=45
  twistMotion=Locked
 }
]
