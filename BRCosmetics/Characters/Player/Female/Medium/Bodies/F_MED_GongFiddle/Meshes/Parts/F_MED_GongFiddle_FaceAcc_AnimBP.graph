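# Animation Blueprint for the GongFiddle female character's face-accessory part, driving secondary motion on its hair bones.
# Poses copied from the parent and Body meshes pass through a Control Rig fed skydive and emote flags.
# Head-space RigidBody sims, switched for emotes, feed chained AnimDynamics hair springs.
# A front-end idle curve layers DefaultSlot animation over hair bones; updates also store gravity overrides using YokeQuick's parameters.
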AnimBlueprint /BRCosmetics/Characters/Player/Female/Medium/Bodies/F_MED_GongFiddle/Meshes/Parts/F_MED_GongFiddle_FaceAcc_AnimBP
target_skeleton=&/F_MED_GongFiddle_FaceAcc_Skeleton
graphs [
 AnimGraph {nodes [
   RigidBody!(ComponentPose, Alpha, OverrideWorldGravity:Vector="0,0,-980", ExternalForce) -> (Pose) {
    bClampLinearTranslationLimitToRefPose=true
    bOverrideWorldGravity=true
    overridePhysicsAsset=&/F_MED_GongFiddle_FaceAcc_Physics
    simulationSpace=BaseBoneSpace
    alphaBoolBlend {blendOption=Linear}
    baseBoneRef {boneName=head}
    overrideWorldGravity {z=-980}
    simSpaceSettings {
     maxAngularAcceleration=1000
     maxAngularVelocity=3
     maxLinearAcceleration=700
     maxLinearVelocity=500
     velocityScaleZ=0
     worldAlpha=1
     externalLinearDragV=<0.1 0.1>
    }
   }
   n1 = Layered blend per bone(
    BasePose=Use cached pose 'Start pose'()
    BlendPoses_0=Slot 'DefaultSlot'()
    BlendWeights_0=Get Custom Idle id Playing Fortnite() -> (Custom Idle id Playing Fortnite:double)
   ) {layerSetup [{branchFilters [
       {boneName=base_Hair_01_back}
       {boneName=base_Hair_01_l}
       {boneName=base_Hair_01_l_back}
       {boneName=base_Hair_01_r}
       {boneName=base_Hair_01_r_back}
       {boneName=Hair_03}
       {boneName=Hair_05}
       {boneName=Hair_12}
       {boneName=Hair_14}
       {boneName=Hair_18}
       {boneName=Hair_20}
      ]}]}
   Output Pose(Result=@n1) {layerGroup=None}
   n2 = Layered blend per bone(
    BasePose=Copy Pose From Mesh() {bUseAttachedParent=true}
    BlendPoses_0=Copy Pose From Mesh(SourceMeshComponent=Get Body() -> (Body:Skeletal Mesh Component Budgeted Object Reference))
   ) {layerSetup [{branchFilters [
       {blendDepth=2, boneName=clavicle_r}
       {blendDepth=2, boneName=clavicle_l}
       {blendDepth=2, boneName=upperarm_l}
       {blendDepth=2, boneName=upperarm_r}
      ]}]}
   n3 = Control Rig!(Source=@n2, Alpha, SkyDive:bool=Get bIsSkyDiving(), Emote:bool=Get bIsPlayingEmote()) -> (Pose) {controlRigAssetReference {blueprintRigClass=&/F_MED_GongFiddle_FaceAcc_CtrlRig.F_MED_GongFiddle_FaceAcc_CtrlRig_C}}
   Save cached pose 'CachPforEmoteVar'(Pose=@n3) {cacheName=CachPforEmoteVar}
   n4 = RigidBody!(
    ComponentPose=Local To Component(LocalPose=Use cached pose 'CachPforEmoteVar'())
    bAlphaBoolEnabled:bool
    OverrideWorldGravity="0,0,-980"
    ExternalForce
    SimSpaceSettings:Sim Space Settings Structure=(WorldAlpha=1,VelocityScaleZ=0,DampingAlpha=1,MaxLinearVelocity=500,MaxAngularVelocity=3,MaxLinearAcceleration=700,MaxAngularAcceleration=1000,ExternalLinearDragV=(X=0.1,Y=0.1,Z=0),ExternalLinearVelocity=(X=0,Y=0,Z=0),ExternalAngularVelocity=(X=0,Y=0,Z=0))
   ) -> (Pose) {
    alphaInputType=Bool
    bClampLinearTranslationLimitToRefPose=true
    bOverrideWorldGravity=true
    overridePhysicsAsset=&/F_MED_GongFiddle_FaceAcc_Physics_Emote
    simulationSpace=BaseBoneSpace
    alphaBoolBlend {blendOption=Linear}
    baseBoneRef {boneName=head}
    overrideWorldGravity {z=-980}
    simSpaceSettings {
     maxAngularAcceleration=1000
     maxAngularVelocity=3
     maxLinearAcceleration=700
     maxLinearVelocity=500
     velocityScaleZ=0
     worldAlpha=1
     externalLinearDragV=<0.1 0.1>
    }
   }
   n5 = Component To Local(ComponentPose=@n4)
   n6 = RigidBody!(
    ComponentPose=Local To Component(LocalPose=Use cached pose 'CachPforEmoteVar'())
    bAlphaBoolEnabled
    OverrideWorldGravity="0,0,-980"
    ExternalForce
   ) -> (Pose) {
    alphaInputType=Bool
    bClampLinearTranslationLimitToRefPose=true
    bOverrideWorldGravity=true
    cachedBoundsScale=1
    overridePhysicsAsset=&F_MED_GongFiddle_FaceAcc_Physics
    simulationSpace=BaseBoneSpace
    alphaBoolBlend {blendOption=Linear}
    baseBoneRef {boneName=head}
    simSpaceSettings {dampingAlpha=0, maxAngularAcceleration=500, maxAngularVelocity=3, maxLinearAcceleration=1100, maxLinearVelocity=300, worldAlpha=1}
   }
   n7 = Component To Local(ComponentPose=@n6)
   n8 = Get bIsPlayingEmote()
   n9 = Anim Dynamics - Bone: Hair_14(
    ComponentPose=Local To Component(LocalPose=Blend Poses by bool(BlendPose_0=@n5, BlendPose_1=@n7, bActiveValue=@n8) {blendType=Linear})
    Alpha=0.2
    ExternalForce="0,0,0"
   ) {
    angularDampingOverride=1
    angularSpringConstant=500
    bAngularSpring=true
    bOverrideAngularDamping=true
    bUseGravityOverride=true
    gravityScale=0.35
    simulationSpace=RootRelative
    boundBone {boneName=Hair_14}
    componentAppliedLinearAccClamp=<5000 5000 5000>
    componentLinearAccScale=<0.6 0.6 0.6>
    componentLinearVelScale=<0.6 0.6 0.6>
    gravityOverride {z=-980}
    physicsBodyDefinitions [{
      sphereCollisionRadius=1.3
      boundBone {boneName=Hair_14}
      boxExtents=<50 50 50>
      constraintSetup {angularLimitsMax=<0 3 0>, angularLimitsMin=<0 0 0>, angularTarget=<1 0 0>}
      localJointOffset=<-1 0 0>
     }]
    relativeSpaceBone {boneName=head}
   }
   n10 = Anim Dynamics - Bone: Hair_15(ComponentPose=@n9, Alpha=0.2, ExternalForce="0,0,0") {
    angularDampingOverride=1
    angularSpringConstant=500
    bAngularSpring=true
    bOverrideAngularDamping=true
    bUseGravityOverride=true
    gravityScale=0.35
    simulationSpace=RootRelative
    boundBone {boneName=Hair_15}
    componentAppliedLinearAccClamp=<5000 5000 5000>
    componentLinearAccScale=<0.6 0.6 0.6>
    componentLinearVelScale=<0.6 0.6 0.6>
    gravityOverride {z=-980}
    physicsBodyDefinitions [{
      sphereCollisionRadius=1.3
      boundBone {boneName=Hair_15}
      boxExtents=<50 50 50>
      constraintSetup {angularLimitsMax=<0 10 0>, angularLimitsMin=<0 0 0>, angularTarget=<1 0 0>}
      localJointOffset=<-1 0 0>
     }]
    relativeSpaceBone {boneName=head}
   }
   n11 = Anim Dynamics - Bone: Hair_16(ComponentPose=@n10, Alpha=0.2, ExternalForce="0,0,0") {
    angularDampingOverride=1
    angularSpringConstant=500
    bAngularSpring=true
    bOverrideAngularDamping=true
    bUseGravityOverride=true
    gravityScale=0.35
    simulationSpace=RootRelative
    boundBone {boneName=Hair_16}
    componentAppliedLinearAccClamp=<5000 5000 5000>
    componentLinearAccScale=<0.6 0.6 0.6>
    componentLinearVelScale=<0.6 0.6 0.6>
    gravityOverride {z=-980}
    physicsBodyDefinitions [{
      sphereCollisionRadius=1.3
      boundBone {boneName=Hair_16}
      boxExtents=<50 50 50>
      constraintSetup {angularLimitsMax=<0 10 2>, angularLimitsMin=<0 -1 -2>, angularTarget=<1 0 0>}
      localJointOffset=<-1 0 0>
     }]
    relativeSpaceBone {boneName=head}
   }
   n12 = Anim Dynamics - Bone: Hair_05(ComponentPose=@n11, Alpha=1, ExternalForce="0,0,0") {
    angularDampingOverride=1
    angularSpringConstant=500
    bAngularSpring=true
    bOverrideAngularDamping=true
    bUseGravityOverride=true
    gravityScale=0.35
    simulationSpace=RootRelative
    boundBone {boneName=Hair_05}
    componentAppliedLinearAccClamp=<5000 5000 5000>
    componentLinearAccScale=<0.6 0.6 0.6>
    componentLinearVelScale=<0.6 0.6 0.6>
    gravityOverride {z=-980}
    physicsBodyDefinitions [{
      sphereCollisionRadius=1.3
      boundBone {boneName=Hair_05}
      boxExtents=<20 20 20>
      constraintSetup {angularLimitsMax=<0 0 0>, angularLimitsMin=<0 -7 -15>, angularTarget=<1 0 0>}
      localJointOffset=<1 0 0>
     }]
    relativeSpaceBone {boneName=head}
   }
   n13 = Anim Dynamics - Bone: Hair_06(ComponentPose=@n12, Alpha=1, ExternalForce="0,0,0") {
    angularDampingOverride=1
    angularSpringConstant=500
    bAngularSpring=true
    bOverrideAngularDamping=true
    bUseGravityOverride=true
    gravityScale=0.35
    simulationSpace=RootRelative
    boundBone {boneName=Hair_06}
    componentAppliedLinearAccClamp=<5000 5000 5000>
    componentLinearAccScale=<0.6 0.6 0.6>
    componentLinearVelScale=<0.6 0.6 0.6>
    gravityOverride {z=-980}
    physicsBodyDefinitions [{
      sphereCollisionRadius=1.3
      boundBone {boneName=Hair_06}
      boxExtents=<20 20 20>
      constraintSetup {angularLimitsMax=<0 0 2>, angularLimitsMin=<0 -7 -15>, angularTarget=<1 0 0>}
      localJointOffset=<1 0 0>
     }]
    relativeSpaceBone {boneName=head}
   }
   n14 = Anim Dynamics - Bone: Hair_19(ComponentPose=@n13, Alpha=1, ExternalForce="0,0,0") {
    angularDampingOverride=1
    angularSpringConstant=500
    bAngularSpring=true
    bOverrideAngularDamping=true
    bUseGravityOverride=true
    gravityScale=0.35
    simulationSpace=RootRelative
    boundBone {boneName=Hair_19}
    componentAppliedLinearAccClamp=<5000 5000 5000>
    componentLinearAccScale=<0.6 0.6 0.6>
    componentLinearVelScale=<0.6 0.6 0.6>
    gravityOverride {z=-980}
    physicsBodyDefinitions [{
      sphereCollisionRadius=1.3
      boundBone {boneName=Hair_19}
      boxExtents=<20 20 20>
      constraintSetup {angularLimitsMax=<0 1 10>, angularLimitsMin=<0 -10 -5>, angularTarget=<1 0 0>}
      localJointOffset=<1 0 0>
     }]
    relativeSpaceBone {boneName=head}
   }
   n15 = Anim Dynamics - Bone: Hair_12(ComponentPose=@n14, Alpha=1, ExternalForce="0,0,0") {
    angularDampingOverride=1
    angularSpringConstant=500
    bAngularSpring=true
    bOverrideAngularDamping=true
    bUseGravityOverride=true
    gravityScale=0.35
    simulationSpace=RootRelative
    boundBone {boneName=Hair_12}
    componentAppliedLinearAccClamp=<5000 5000 5000>
    componentLinearAccScale=<0.6 0.6 0.6>
    componentLinearVelScale=<0.6 0.6 0.6>
    gravityOverride {z=-980}
    physicsBodyDefinitions [{
      sphereCollisionRadius=1.3
      boundBone {boneName=Hair_12}
      boxExtents=<20 20 20>
      constraintSetup {angularLimitsMax=<0 0 5>, angularLimitsMin=<0 -10 -5>, angularTarget=<1 0 0>}
      localJointOffset=<1 0 0>
     }]
    relativeSpaceBone {boneName=head}
   }
   n16 = Anim Dynamics - Bone: Hair_13(ComponentPose=@n15, Alpha=1, ExternalForce="0,0,0") {
    angularDampingOverride=1
    angularSpringConstant=500
    bAngularSpring=true
    bOverrideAngularDamping=true
    bUseGravityOverride=true
    gravityScale=0.35
    simulationSpace=RootRelative
    boundBone {boneName=Hair_13}
    componentAppliedLinearAccClamp=<5000 5000 5000>
    componentLinearAccScale=<0.6 0.6 0.6>
    componentLinearVelScale=<0.6 0.6 0.6>
    gravityOverride {z=-980}
    physicsBodyDefinitions [{
      sphereCollisionRadius=1.3
      boundBone {boneName=Hair_13}
      boxExtents=<20 20 20>
      constraintSetup {angularLimitsMax=<0 0 5>, angularLimitsMin=<0 -10 -5>, angularTarget=<1 0 0>}
      localJointOffset=<1 0 0>
     }]
    relativeSpaceBone {boneName=head}
   }
   Save cached pose 'Start pose'(Pose=Component To Local(ComponentPose=@n16)) {cacheName=Start pose}
  ]}
 EventGraph {nodes [
   Event BlueprintUpdateAnimation()
   > n1 = "Gravity Override State Function (Parameter Asset) (Nativized)"(
    CharacterPartAnimInstance=Self-Reference()
    GameTimeInSeconds=GetGameTimeInSeconds()
    DebugJointName=spine_05
    Parameters=/BRCosmetics/Characters/Player/Male/Medium/Bodies/M_MED_YokeQuick/Meshes/M_MED_YokeQuick_Parameters
   )
   > Set Gravity Override(Gravity Override=@n1)
   > Set RBAN Settings Out(RBAN Settings Out=@n1.OutRigidBodyAnimNodeParameters)
   > Branch(Condition=Get bIsInFrontEnd())
   > Set Custom Idle id Playing Fortnite(Custom Idle id Playing Fortnite:double=GetCurveValue(self=GetOwnerFrontEndAnimInstance(), CurveName=Frontend_Idle_Bypass)) -> (Output_Get:double)
   n2 = Event OnCharacterCustomizationLoaded()
   > Set Body(Body:Skeletal Mesh Component Budgeted Object Reference=Get Array Item with EFort Custom Part Type(Array=@n2.CustomMeshParts)) -> (Output_Get:Skeletal Mesh Component Budgeted Object Reference)
  ]}
]
variables [
 Body {type=SkeletalMeshComponentBudgeted}
 Custom Idle id Playing Fortnite {type=number}
 Gravity Override {type=Vector}
 RBAN Settings Out {
  type=FortRigidBodyAnimNodeParameters
  default {simSpaceSettings {worldAlpha=1, externalAngularVelocity=<1 1 1>, externalLinearDragV=<1 1 1>, externalLinearVelocity=<1 1 1>}}
 }
]
